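# Niagara trail effect for the SurgeRaven glider, combining exhaust flames, lights, spinning tire meshes and speed lines.
# Flames and lights follow skeletal mesh sockets, while tires use position offsets and speed lines spawn on a torus.
# System Timeline values ramp flame size and brightness, tire spin and speed-line density.
# User ForwardDot adjusts light alpha, RightDot skews speed lines, and front-end modifiers swap values for non-gameplay display.
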NiagaraSystem /BRCosmetics/Gadgets/Assets/VinderTech_GliderChute/Glider_SurgeRaven/FX/NS_Glider_SurgeRaven_Trail
dependencies {
 usedDataInterfaces [
  &NiagaraDataInterfaceCurve
  &NiagaraDataInterfaceArrayInt32
  &NiagaraDataInterfaceArrayFloat3
  &NiagaraDataInterfaceSkeletalMesh
  &NiagaraDataInterfaceArrayFloat
 ]
 usedDynamicInputs [
  &RandomRangeInt
  &SelectFloatFromArray
  &MakeIntFromFloat
  &NormalizeVector
  &TransformVector
  &InvertBool
  &FNTimeOfDayColor
 ]
 usedModules [
  &self:SystemSpawnScript.SystemScriptSource.SystemScriptGraph.NiagaraNodeAssignment_0.SetVariables_#1
  &SystemState
  &Timeline
  &FN_Integrator
  &InterpolateOverTime
  &self:SystemSpawnScript.SystemScriptSource.SystemScriptGraph.NiagaraNodeAssignment_1.SetVariables_#2
  &self:SystemSpawnScript.SystemScriptSource.SystemScriptGraph.NiagaraNodeAssignment_2.SetVariables_#3
  &self:SystemSpawnScript.SystemScriptSource.SystemScriptGraph.NiagaraNodeAssignment_3.SetVariables_#4
 ]
}
emitters {
 NE_Flames {
  emitterUpdateScript [
   &EmitterState(Life Cycle Mode=::NewEnumerator1, Loop Behavior=::NewEnumerator1, Loop Delay=2.3, Loop Duration Mode=::NewEnumerator1, UseLoopDelay=true)
   &SpawnBurst_Instantaneous(Spawn Count=4, Use Spawn Probability=true)
   /SetVariables(
    Emitter.Flame1A:LinearColor=<rgba 0.125 0.0688 0.0125 1>
    Emitter.Flame1B:LinearColor=<rgba 0.5 0.4333 0.1 1>
    Emitter.Flame2A:LinearColor=<rgba 1 0.6667 0.2 1>
    Emitter.Flame2B:LinearColor=<rgba 0.4 0.2133 0.08 1>
    Emitter.Flame3A:LinearColor=<rgba 1 0.4838 0.1 1>
    Emitter.Flame3B:LinearColor=<rgba 0.4583 0.2796 0.0458 1>
    Emitter.FlameCoreA:LinearColor=<rgba 1 0.8461 0.2 1>
    Emitter.FlameCoreB:LinearColor=<rgba 1 0.7 0.1 1>
    Emitter.FlameDark:LinearColor=<rgba 0.1927 0.1172 0.0416 1>
    Emitter.FlameLight:LinearColor=<rgba 0.3854 0.2318 0.0781 1>
   ) {isSetParameters=true}
   &SpawnBurst_Instantaneous(Spawn Count=2, Spawn Group=2, Use Spawn Probability=false, UseLoopCountLimit=false) {module=SpawnBurst_Instantaneous001}
  ]
  particleSpawnScript [
   &PartitionParticles(ID to Partition:NiagaraInt32=$Engine.ExecIndex)
   /SetVariables(Particles.VisibilityTag=&SpawnGroup()) {isSetParameters=true}
   &InitializeParticle(
    Color=&MergeHelper_ColorSwitch(0=<rgba 1 1 1 1>, 1=<rgba 1 1 1 1>, 2=<rgba 5 2.1942 0.4043 1>, 3=?, Index=$Particles.VisibilityTag, Num=::NewEnumerator1)
    Lifetime=0.15
    Mesh Scale=<0.14 0.14 0.3>
    Mesh Scale Mode=::NewEnumerator4
    Sprite Size Mode=::NewEnumerator3
    Uniform Sprite Size=35
    UsePositionOffset=false
    Write Material Random=true
   )
   &InitialMeshOrientation(
    Orientation Axis=&MergeHelper_VectorSwitch(0=<-1 0 0>, 1=<0 0 0>, 2=?, 3=?, Index=$Particles.PartitionParticles.Partition, Num=::NewEnumerator0)
    Orientation Vector=<0 0 1>
    Use Orientation Vector=true
    Use Rotation Vector=false
   ) {module=InitialMeshOrientation001}
   /SetVariables(
    Particles.Flame2Scale:Vector3f=&Multiply_Vector(A=$Particles.Scale, B=<1 1 0.9>)
    Particles.FlameCoreScale:Vector3f=&Multiply_Vector(A=$Particles.Scale, B=<0.65 0.6 0.6>)
   ) {isSetParameters=true}
   &CameraOffset(Camera Offset Amount=&MergeHelper_FloatSwitch(
     0=0
     1=0
     2=&FN_FrontEndFloatModifier(Gameplay=5, Front End=0)
     3=?
     Index=$Particles.VisibilityTag
     Num=::NewEnumerator1
    ))
  ]
  particleUpdateScript [
   &ParticleState(Kill Particles When Lifetime Has Elapsed=false, Let Infinitely Lived Particles Die When Emitter Deactivates=true)
   &SkeletalMeshLocation(
    Apply Sampled Orientation to Mesh Particles=true
    Bone / Socket Index=&SelectIntFromArray(Direct Array Index=$Particles.PartitionParticles.Partition, Int Selection Array=DataInterface() {propertyValues {IntData [0,1]}})
    Inherit Velocity=false
    MaskSpawnGroup=false
    Mesh Sampling Type=::NewEnumerator5
    Sampled Position Offset=&SelectVectorFromArray(Direct Array Index=$Particles.PartitionParticles.Partition, Vector Selection Array=DataInterface() {propertyValues {FloatData []}})
    Sampled Position Offset Coordinate Space:ENiagaraMeshTransforms=::NewEnumerator3
    Skeletal Mesh=$System.Skeletal Mesh
    Socket Sampling Mode=::NewEnumerator5
   )
   &SolveForcesAndVelocity(Clamp Velocity=false, Limit Acceleration=false)
   &ScaleColor(
    Scale Alpha=&FloatFromCurve(
     FloatCurve=DataInterface() {propertyValues {Curve {keys [<at 0 0 RCIM_Cubic>,<at 0.5 1 RCIM_Cubic>]}}}
     Scale Curve=&FN_FrontEndFloatModifier(
      Gameplay=&MergeHelper_FloatSwitch(
       0=&Lerp_Float(A=0, B=1, Alpha=$System.Timeline.Time)
       1=&Lerp_Float(A=0, B=1, Alpha=$System.Timeline.Time)
       2=&Lerp_Float(A=0.5, B=1, Alpha=$System.Timeline.Time)
       3=?
       Index=$Particles.VisibilityTag
       Num=::NewEnumerator1
      )
      Front End=0.5
     )
    )
    Scale RGB=&MakeVectorFromLinearColorRGB(Color=&MergeHelper_ColorSwitch(
      0=&LerpLinearColors(StartColor=<rgba 1 1 1 1>, EndColor=<rgba 1.5 1.5 1.5 1>, LerpFactor=$System.Timeline.Time)
      1=&LerpLinearColors(StartColor=<rgba 1 1 1 1>, EndColor=<rgba 1.5 1.5 1.5 1>, LerpFactor=$System.Timeline.Time)
      2=&LerpLinearColors(StartColor=<rgba 1 1 1 1>, EndColor=<rgba 1.5 1.5 1.5 1>, LerpFactor=$System.Timeline.Time)
      3=?
      Index=$Particles.VisibilityTag
      Num=::NewEnumerator1
     ))
    ScaleA=true
    ScaleRGB=true
   )
   /SetVariables(
    Particles.Flame2Color:LinearColor=&MakeLinearColorFromVectorAndFloat("Vector (RGB)"=<1 1 1>, "Float (Alpha)"=&Lerp_Float(A=1, B=5, Alpha=$System.Timeline.Time))
    Particles.FlameCoreColor:LinearColor=&MakeLinearColorFromVectorAndFloat("Vector (RGB)"=&MakeVectorFromLinearColorRGB(Color=<rgba 10 3.6336 0 1>), "Float (Alpha)"=0.3)
   ) {isSetParameters=true}
   &ScaleMeshSize(Scale Factor=&MakeVector(
     X=&FN_FrontEndFloatModifier(Gameplay=&Lerp_Float(A=0.5, B=1, Alpha=$System.Timeline.Time), Front End=1)
     Y=&FN_FrontEndFloatModifier(Gameplay=&Lerp_Float(A=0.5, B=1, Alpha=$System.Timeline.Time), Front End=1)
     Z=&FN_FrontEndFloatModifier(Gameplay=&Lerp_Float(A=0.5, B=2, Alpha=$System.Timeline.Time), Front End=2)
    ))
   /SetVariables(
    Particles.Flame2Scale:Vector3f=&Multiply_Vector(
     A=$Particles.Initial.Flame2Scale
     B=&MakeVector(
      X=&FN_FrontEndFloatModifier(Gameplay=&Lerp_Float(A=0.5, B=1, Alpha=$System.Timeline.Time), Front End=1)
      Y=&FN_FrontEndFloatModifier(Gameplay=&Lerp_Float(A=0.5, B=1, Alpha=$System.Timeline.Time), Front End=1)
      Z=&FN_FrontEndFloatModifier(Gameplay=&Lerp_Float(A=0.5, B=2, Alpha=$System.Timeline.Time), Front End=2)
     )
    )
    Particles.FlameCoreScale:Vector3f=&Multiply_Vector(
     A=$Particles.Initial.FlameCoreScale
     B=&MakeVector(
      X=&FN_FrontEndFloatModifier(Gameplay=&Lerp_Float(A=0.5, B=1, Alpha=$System.Timeline.Time), Front End=1)
      Y=&FN_FrontEndFloatModifier(Gameplay=&Lerp_Float(A=0.5, B=1, Alpha=$System.Timeline.Time), Front End=1)
      Z=&FN_FrontEndFloatModifier(Gameplay=&Lerp_Float(A=0.5, B=2, Alpha=$System.Timeline.Time), Front End=1)
     )
    )
   ) {isSetParameters=true}
   &DynamicMaterialParameters(
    Index 0 Param 1=&Lerp_Float(A=0.3, B=0.6, Alpha=1)
    Module.Param0WriteEnabled=true
    Module.Param1WriteEnabled=true
    Module.Param2WriteEnabled=true
    Module.Param3WriteEnabled=true
   )
  ]
  rendererClasses [
   &NiagaraMeshRendererProperties
   &NiagaraSpriteRendererProperties
   &NiagaraLightRendererProperties
  ]
  renderers [
   {rendererClass=&NiagaraMeshRendererProperties, rendererIndex=0}
   {rendererClass=&NiagaraMeshRendererProperties, rendererIndex=1}
   {rendererClass=&NiagaraMeshRendererProperties, rendererIndex=2}
   {rendererClass=&NiagaraMeshRendererProperties, rendererIndex=3}
   {rendererClass=&NiagaraSpriteRendererProperties, rendererIndex=4}
   {rendererClass=&NiagaraLightRendererProperties, rendererIndex=5}
   {rendererClass=&NiagaraMeshRendererProperties, rendererIndex=6}
  ]
 }
 NE_Lights {
  emitterUpdateScript [
   &EmitterState(Life Cycle Mode=::NewEnumerator1, Loop Behavior=::NewEnumerator1, Loop Delay=0.5, Loop Duration Mode=::NewEnumerator1, UseLoopDelay=true)
   &SpawnBurst_Instantaneous(Spawn Probability=HlslExpression() {hlslExpression="Emitter.LoopCount == 0"}, Use Spawn Probability=true, UseLoopCountLimit=false)
   &SpawnBurst_Instantaneous(
    Spawn Group=1
    Spawn Probability=HlslExpression() {hlslExpression="Emitter.LoopCount == 0"}
    Use Spawn Probability=true
    UseLoopCountLimit=false
   ) {module=SpawnBurst_Instantaneous001}
   &SpawnBurst_Instantaneous(
    Spawn Group=2
    Spawn Probability=HlslExpression() {hlslExpression="Emitter.LoopCount == 0"}
    Use Spawn Probability=true
    UseLoopCountLimit=false
   ) {module=SpawnBurst_Instantaneous002}
  ]
  particleSpawnScript [
   /SetVariables(Particles.VisibilityTag=&SpawnGroup()) {isSetParameters=true}
   &InitializeParticle(
    Color=&MergeHelper_ColorSwitch(0=<rgba 0 0 0 0>, 1=<rgba 1 0 0.0421 1>, 2=<rgba 1 0 0.0501 1>, 3=?, Index=$Particles.VisibilityTag, Num=::NewEnumerator1)
    Lifetime=0.5
    Mesh Scale Mode=::NewEnumerator3
    Mesh Uniform Scale=0.9
    Sprite Size Mode=::NewEnumerator3
    Uniform Sprite Size=&MergeHelper_FloatSwitch(
     0=0
     1=&FN_FrontEndFloatModifier(Gameplay=100, Front End=50)
     2=&FN_FrontEndFloatModifier(Gameplay=200, Front End=125)
     3=?
     Index=$Particles.VisibilityTag
     Num=::NewEnumerator1
    )
    UsePositionOffset=false
    Write Material Random=true
   )
   /SetVariables(Particles.LightMeshColor:LinearColor=<rgba 13.0435 14.7448 15 0.2>, Particles.LightMeshOuterColor:LinearColor=<rgba 1.5652 2 1.9433 0.5>) {isSetParameters=true}
   /SetVariables(Particles.LightMeshScale:Vector3f=<3 2 2>) {isSetParameters=true}
   /SetVariables(Particles.FrontEndMult=0.75) {isSetParameters=true}
   &SpriteFacingAndAlignment(
    Alignment Coordinate Space=::Local
    Facing Coordinate Space=::Local
    Sprite Alignment=<0 0 1>
    Sprite Facing=<0 1 0>
    WriteAlignment=true
    WriteFacing=true
   )
   &CameraOffset(Camera Offset Amount=&MergeHelper_FloatSwitch(0=0, 1=0, 2=10, 3=?, Index=$Particles.VisibilityTag, Num=::NewEnumerator1))
  ]
  particleUpdateScript [
   &ParticleState(Kill Particles When Lifetime Has Elapsed=false)
   &SkeletalMeshLocation(
    Bone / Socket Index=&SelectIntFromArray(Direct Array Index=$Particles.VisibilityTag, Int Selection Array=DataInterface() {propertyValues {IntData [3,2,2]}})
    MaskSpawnGroup=false
    Mesh Sampling Type=::NewEnumerator5
    Sampled Position Offset=&FN_FrontEndVector3Modifier(
     Gameplay=&SelectVectorFromArray(
      Direct Array Index=$Particles.VisibilityTag
      Vector Selection Array=DataInterface() {propertyValues {FloatData [<0 0 0>, <0 -2 0>, <0 -2 0>]}}
     )
     Front End=&SelectVectorFromArray(Direct Array Index=$Particles.VisibilityTag, Vector Selection Array=DataInterface() {propertyValues {FloatData [<0 0 0>, <0 0 0>, <0 0 0>]}})
    )
    Skeletal Mesh=$System.Skeletal Mesh
    Socket Sampling Mode=::NewEnumerator5
   ) {module=SkeletalMeshLocation001}
   /SetVariables(Particles.Flicker:NiagaraFloat=&FloatFromCurve(FloatCurve=DataInterface() {propertyValues {Curve {keys [<at 0 0.75>,<at 1 0.75>]}}})) {isSetParameters=true}
   /SetVariables(
    Particles.LightMeshColor:LinearColor=&Multiply_LinearColorByFloat(Linear Color=$Particles.Initial.LightMeshColor, Float=$Particles.Flicker)
    Particles.LightMeshOuterColor:LinearColor=&Multiply_LinearColorByFloat(Linear Color=$Particles.Initial.LightMeshOuterColor, Float=$Particles.Flicker)
   ) {isSetParameters=true}
   /SetVariables(
    Particles.LightMeshColor:LinearColor=&Multiply_LinearColorByFloat(
     Linear Color=$Particles.LightMeshColor
     Float=&FN_FrontEndFloatModifier(Gameplay=1, Front End=$Particles.FrontEndMult)
    )
    Particles.LightMeshOuterColor:LinearColor=&Multiply_LinearColorByFloat(
     Linear Color=$Particles.LightMeshOuterColor
     Float=&FN_FrontEndFloatModifier(Gameplay=1, Front End=$Particles.FrontEndMult)
    )
   ) {isSetParameters=true}
   &ScaleColor(
    Scale Alpha=&MergeHelper_FloatSwitch(
     0=0
     1=1
     2=&Lerp_Float(
      A=0.5
      B=1
      Alpha=&MakeCustomFloatFromBool(
       Bool=&SetBoolByFloatComparison(A=$User.ForwardDot, B=-1, Comparison Type=::NewEnumerator4)
       True Float=1
       False Float=0
      )
     )
     3=?
     Index=$Particles.VisibilityTag
     Num=::NewEnumerator1
    )
    ScaleA=true
    ScaleRGB=true
   )
  ]
  rendererClasses [&NiagaraMeshRendererProperties, &NiagaraSpriteRendererProperties]
  renderers [
   {rendererClass=&NiagaraMeshRendererProperties, rendererIndex=0}
   {rendererClass=&NiagaraMeshRendererProperties, rendererIndex=1}
   {rendererClass=&NiagaraMeshRendererProperties, rendererIndex=2}
   {rendererClass=&NiagaraSpriteRendererProperties, rendererIndex=3}
   {rendererClass=&NiagaraSpriteRendererProperties, rendererIndex=4}
  ]
 }
 NE_SpeedLines {
  emitterUpdateScript [
   &EmitterState(Life Cycle Mode=::NewEnumerator1, Loop Behavior=::NewEnumerator1, Loop Delay=2.5, Loop Duration Mode=::NewEnumerator1, UseLoopDelay=true)
   &SpawnRate(
    SpawnGroup=1
    SpawnRate=&Lerp_Float(
     A=&FN_FrontEndFloatModifier(
      Gameplay=0
      Front End=&FloatFromCurve(FloatCurve=DataInterface() {propertyValues {Curve {keys [<at 0.5 0>,<at 2 1>]}}}, CurveIndex=$Engine.System.Age, Scale Curve=75)
     )
     B=200
     Alpha=$System.Timeline.NormalizedTime
    )
    Use Spawn Probability=true
   ) {module=SpawnRate001}
  ]
  particleSpawnScript [
   &InitializeParticle(
    Color=<rgba 1 1 1 5>
    Mesh Scale Max=<50 1 1>
    Mesh Scale Min=<15 1 1>
    Mesh Scale Mode=::NewEnumerator2
    UsePositionOffset=false
    Write Material Random=true
   )
   &InitialMeshOrientation(
    Mesh Orientation Mode=::NewEnumerator2
    Rotation=&RandomRangeVector(Minimum=<0 0 0>, Maximum=<1 0 0>)
    Use Rotation Vector=true
   )
   &TorusLocation(
    Align Torus With Axis=true
    Handle Radius=0
    Large Radius:NiagaraFloat=&FN_FrontEndFloatModifier(
     Gameplay=&RandomRangeFloat(Minimum=180, Maximum=220)
     Front End=&RandomRangeFloat(Minimum=180, Maximum=220)
    )
    MaskSpawnGroup=false
    Module.Override Randomness=false
    Offset=<0 -180 0>
    Spawn Group Mask=1
    Torus Axis=<0 1 0>
    Torus Origin:NiagaraPosition=&Add_Vector(
     A=&FN_FrontEndVector3Modifier(Gameplay=<0 -100 50>, Front End=<0 -250 -100>)
     B=&Lerp_Vector(A=<0 0 0>, B=<200 0 0>, Alpha=$User.RightDot)
    )
   ) {module=TorusLocation001}
   &AddVelocityFromPoint(
    Module.Use Falloff Distance=false
    Origin Offset=&Add_Vector(A=<0 -500 0>, B=&Lerp_Vector(A=<0 0 0>, B=<300 0 0>, Alpha=$User.RightDot))
    Velocity Strength=&FN_FrontEndFloatModifier(Gameplay=150, Front End=150)
   )
   &AddVelocity(
    Module.Use Falloff Distance=false
    Origin Offset=&Add_Vector(A=<0 -500 0>, B=&Lerp_Vector(A=<0 0 0>, B=<500 0 0>, Alpha=$User.RightDot))
    Override Seed=false
    Velocity Mode=::NewEnumerator1
    Velocity Speed=200
   ) {module=AddVelocity001}
   &AddVelocity(Velocity=&Multiply_Vector(
     A=&MakeCustomVectorFromBool(
      Bool=$User.bIsFrontEndPreview
      True Vector=&RandomRangeVector(Minimum=<0 500 1000>, Maximum=<0 500 1000>)
      False Vector=&RandomRangeVector(Minimum=<1 400 1>, Maximum=<1 600 1>)
     )
     B=&MakeVector(X=1, Y=&MakeCustomFloatFromBool(Bool=$User.bIsFrontEndPreview, True Float=5, False Float=1), Z=1)
    ))
  ]
  particleUpdateScript [
   &ScaleVelocity(Velocity Scale=&VectorFromFloat(Value=&FloatFromCurve(
      FloatCurve=DataInterface() {propertyValues {Curve {keys [
          {interpMode=RCIM_Cubic, tangentMode=RCTM_User, time=0, value=0}
          {arriveTangent=2, interpMode=RCIM_Cubic, leaveTangent=2, time=0.1, value=0.975}
          {interpMode=RCIM_Cubic, tangentMode=RCTM_User, time=0.5, value=1}
          {arriveTangent=-2, interpMode=RCIM_Cubic, leaveTangent=-2, time=0.9, value=1}
          {tangentMode=RCTM_User, time=1, value=0}
         ]}}}
      CurveIndex=&RemapRange(Input Value To Remap=$User.RightDot, Input Min=-1, Input Max=1, Output Min=0, Output Max=1)
     )))
   &SolveForcesAndVelocity(Clamp Velocity=false, Limit Acceleration=false)
   &ParticleState()
   &ScaleMeshSize(Scale Factor=&VectorFromFloat(Value=&FloatFromCurve(FloatCurve=DataInterface() {propertyValues {Curve {keys [
          {tangentMode=RCTM_User, time=0, value=1}
          {arriveTangent=-0.0405, interpMode=RCIM_Cubic, leaveTangent=-0.0405, tangentMode=RCTM_User, time=0.7009, value=1}
          {arriveTangent=-8.0089, interpMode=RCIM_Cubic, leaveTangent=-8.0089, tangentMode=RCTM_User, time=1, value=0}
         ]}}})))
   &DynamicMaterialParameters(
    Index 0 Param 2=1
    Index 0 Param 3=1
    Index 0 Param 4=1
    Module.Param0WriteEnabled=true
    Module.Param1WriteEnabled=true
    Module.Param2WriteEnabled=true
    Module.Param3WriteEnabled=true
   )
   &ScaleColor(
    Scale Alpha=&MakeCustomFloatFromBool(
     Bool=$User.bIsFrontEnd
     True Float=&FloatFromCurve(FloatCurve=DataInterface() {propertyValues {Curve {keys [<at 0 1>,<at 1 0>]}}})
     False Float=&FloatFromCurve(
      FloatCurve=DataInterface() {propertyValues {Curve {keys [
          {interpMode=RCIM_Cubic, tangentMode=RCTM_User, time=0, value=0}
          {interpMode=RCIM_Cubic, tangentMode=RCTM_User, time=0.5, value=1.5}
          {tangentMode=RCTM_User, time=1, value=0}
         ]}}}
      Scale Curve=&Lerp_Float(
       A=&FN_FrontEndFloatModifier(Gameplay=0, Front End=0.5)
       B=&Lerp_Float(A=0, B=0.5, Alpha=$System.Timeline001.Time)
       Alpha=$System.Timeline.Time
      )
     )
    )
    ScaleA=true
    ScaleRGB=true
   )
   &ScaleColor(Scale Alpha=&FN_FrontEndFloatModifier(Gameplay=1, Front End=0.2), ScaleA=true, ScaleRGB=true) {module=ScaleColor001}
  ]
  rendererClasses [&NiagaraMeshRendererProperties]
  renderers [{rendererClass=&NiagaraMeshRendererProperties, rendererIndex=0}]
 }
 NE_Tires {
  emitterUpdateScript [
   &EmitterState(Delay First Loop Only=true, Life Cycle Mode=::NewEnumerator1, Loop Delay=2.6, Loop Duration=1, UseLoopDelay=true)
   &SpawnBurst_Instantaneous(
    Spawn Count=3
    Spawn Probability=HlslExpression() {hlslExpression="Emitter.LoopCount == 0"}
    Use Spawn Probability=true
    UseLoopCountLimit=false
   )
   &SpawnBurst_Instantaneous(
    Spawn Count=3
    Spawn Group=1
    Spawn Probability=HlslExpression() {hlslExpression="Emitter.LoopCount == 0"}
    Use Spawn Probability=true
    UseLoopCountLimit=false
   ) {module=SpawnBurst_Instantaneous001}
  ]
  particleSpawnScript [
   &InitializeParticle(
    Color=<rgba 0.2 0.2 0.2 1>
    Lifetime=5
    Mesh Scale Mode=::NewEnumerator3
    Position Offset=&MergeHelper_VectorSwitch(0=<0 -99 -71>, 1=<0 40 -59>, 2=?, 3=?, Index=&SpawnGroup(), Num=::NewEnumerator0)
    UsePositionOffset=true
    Write Material Random=true
   )
   &InitialMeshOrientation(Mesh Orientation Mode=::NewEnumerator2, Orientation Coordinate Space=::Local, Use Rotation Vector=false)
  ]
  particleUpdateScript [
   &ParticleState(Kill Particles When Lifetime Has Elapsed=false)
   &UpdateMeshOrientation(Rotation Rate=&Lerp_Float(A=3, B=-25, Alpha=$System.Timeline.Time))
   &ScaleColor(
    Scale Alpha=&Lerp_Float(
     A=0
     B=&FNTimeOfDayScalar(Night Scale=0.5, Morning Scale=0.75, Evening Scale=0.75)
     Alpha=$System.Timeline.Time
    )
    ScaleA=true
    ScaleRGB=true
   )
  ]
  rendererClasses [&NiagaraMeshRendererProperties]
  renderers [{rendererClass=&NiagaraMeshRendererProperties, rendererIndex=0}]
 }
}
summary {
 systemName=NS_Glider_SurgeRaven_Trail
 emitters [
  {emitterName=NE_SpeedLines, rendererClasses [&NiagaraMeshRendererProperties]}
  {emitterName=NE_Lights, rendererClasses [&NiagaraMeshRendererProperties, &NiagaraSpriteRendererProperties]}
  {
   emitterName=NE_Flames
   rendererClasses [
    &NiagaraMeshRendererProperties
    &NiagaraSpriteRendererProperties
    &NiagaraLightRendererProperties
   ]
  }
  {emitterName=NE_Tires, rendererClasses [&NiagaraMeshRendererProperties]}
 ]
 userVariables [
  User.bIsFrontEnd {type=NiagaraBool}
  User.bIsFrontEndPreview {type=NiagaraBool}
  User.bIsFullyDeployed {type=NiagaraBool}
  User.ForwardDot {type=NiagaraFloat, defaultValue=0, description=""}
  User.RightDot {type=NiagaraFloat, defaultValue=0, description=""}
 ]
}
